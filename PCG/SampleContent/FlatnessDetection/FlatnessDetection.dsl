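; PCG sample graph that demonstrates flatness detection on a landscape, finding flat areas at several footprint sizes.
; Three surface samplers create candidate points with different extents, tagged Largest, Medium and Small.
; Each set runs through a DiscardPointsOnIrregularSurface subgraph, which tests it against landscape data.
; The surviving points are unioned with a fourth, smallest-extent sampler that skips the test, then converted to points for two Debug nodes.
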(graph FlatnessDetection
  (bind GetLandscapeData_11 (Get%20Landscape%20Data :bComponentsMustOverlapSelf false :SamplingProperties (struct)))
  (bind SurfaceSamplerNode_0 (Surface%20Sampler :Looseness 0.0 :PointExtents (2.5e+02 2.5e+02 2.5e+02) :PointsPerSquaredMeter 1.0 :PointSteepness 0.5))
  (bind AddTags_2 (Add%20Tags :bTokenizeOnWhiteSpace true :TagsToAdd "Largest")) ; Node added to replicate deprecated behavior of the Add Tags To Output functionality.
  (bind GetLandscapeData_20 (Get%20Landscape%20Data :bComponentsMustOverlapSelf false :SamplingProperties (struct)))
  (bind Subgraph_0 (Subgraph :Graph /PCG/Subgraphs/DiscardPointsOnIrregularSurface.DiscardPointsOnIrregularSurface))
  (bind AddTags_0 (Add%20Tags :bTokenizeOnWhiteSpace true :TagsToAdd "Medium")) ; Node added to replicate deprecated behavior of the Add Tags To Output functionality.
  (bind Subgraph_30 (Subgraph :Graph /PCG/Subgraphs/DiscardPointsOnIrregularSurface.DiscardPointsOnIrregularSurface))
  (bind AddTags_1 (Add%20Tags :bTokenizeOnWhiteSpace true :TagsToAdd "Small")) ; Node added to replicate deprecated behavior of the Add Tags To Output functionality.
  (bind Subgraph_31 (Subgraph :Graph /PCG/Subgraphs/DiscardPointsOnIrregularSurface.DiscardPointsOnIrregularSurface))
  (bind Union_28 (Union))
  (bind ToPoint_39 (To%20Point))
  (bind Debug_1 (Debug))
  (bind Debug_35 (Debug))
  ; Points To Test
  (bind SurfaceSamplerNode_2 (Surface%20Sampler :Looseness 0.0 :PointExtents (1e+03 1e+03 2.5e+02) :PointSteepness 0.5))
  (bind SurfaceSampler_8 (Surface%20Sampler :Looseness 0.0 :PointExtents (2e+03 2e+03 2.5e+02) :PointSteepness 0.5))
  (bind SurfaceSampler_1 (Surface%20Sampler :Looseness 0.0 :PointExtents (5e+02 5e+02 2.5e+02) :PointSteepness 0.5))
  (connect SurfaceSamplerNode_2 Out AddTags_0 In)
  (connect SurfaceSampler_1 Out AddTags_1 In)
  (connect SurfaceSampler_8 Out AddTags_2 In)
  (connect ToPoint_39 Out Debug_1 In)
  (connect ToPoint_39 Out Debug_35 In)
  (connect AddTags_2 Out Subgraph_0 Points%20In)
  (connect GetLandscapeData_20 Out Subgraph_0 Surface)
  (connect AddTags_0 Out Subgraph_30 Points%20In)
  (connect GetLandscapeData_20 Out Subgraph_30 Surface)
  (connect AddTags_1 Out Subgraph_31 Points%20In)
  (connect GetLandscapeData_20 Out Subgraph_31 Surface)
  (connect GetLandscapeData_11 Out SurfaceSampler_1 Surface)
  (connect GetLandscapeData_11 Out SurfaceSampler_8 Surface)
  (connect GetLandscapeData_11 Out SurfaceSamplerNode_0 Surface)
  (connect GetLandscapeData_11 Out SurfaceSamplerNode_2 Surface)
  (connect Union_28 Out ToPoint_39 In)
  (connect Subgraph_0 Points%20Out Union_28 In)
  (connect Subgraph_30 Points%20Out Union_28 In)
  (connect Subgraph_31 Points%20Out Union_28 In)
  (connect SurfaceSamplerNode_0 Out Union_28 In))
