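# Translucent, unlit wireframe material in dark blue, part of the PCG flatness detection sample content.
# Its graph holds a blue vector parameter and a Fresnel term with a scalar parameter for the exponent.
# It is flagged for use with instanced static meshes and Nanite.
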Material /PCG/SampleContent/FlatnessDetection/Materials/BlueWireframe
graph {
 properties {
  bUsedWithInstancedStaticMeshes=true
  bUsedWithNanite=true
  blendMode=BLEND_Translucent
  pixelDepthOffsetMode=PDOM_Legacy
  shadingModel=MSM_Unlit
  usedShadingModels=Unlit
  wireframe=true
 }
 nodes [
  Fresnel(ExponentIn=ScalarParameter() {defaultValue=2})
  VectorParameter() {defaultValue=<rgba 0 0.004 0.3073 0>}
 ]
}
